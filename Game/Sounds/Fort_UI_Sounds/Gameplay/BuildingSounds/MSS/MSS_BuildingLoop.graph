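# Stereo MetaSound source for a building loop sound, pairing a start sound with a delayed loop sound.
# On play, a random Start wave fires immediately and a random Loop wave follows after a short delay, each with randomized pitch and decibel gain.
# A mixer combines both players, then BaseVolume scales the output.
# The Mobile page plays only an undelayed mono Loop wave on both channels, leaving the randomized gains unconnected.
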MetaSoundSource /Game/Sounds/Fort_UI_Sounds/Gameplay/BuildingSounds/MSS/MSS_BuildingLoop
inputs [
 BaseVolume {type=Float}
 Character.Affiliation {type=Enum:FortAudioAffiliation}
 Character.Health {type=Float}
 Character.UniqueID {type=Int32}
 Loop.Pitch.Base {type=Float}
 Loop.Pitch.Range {type=Float}
 Loop.Sounds {type=WaveAsset:Array}
 Loop.Volume.Base {type=Float}
 Loop.Volume.Range {type=Float}
 Start.Pitch.Base {type=Float}
 Start.Pitch.Range {type=Float}
 Start.Sounds {type=WaveAsset:Array}
 Start.Volume.Base {type=Float}
 Start.Volume.Range {type=Float}
 UE.Source.OnPlay {type=Trigger}
]
interfaces [Fort.Character:1.0, UE.OutputFormat.Stereo:1.0, UE.Source:1.1]
outputs [
 "UE.OutputFormat.Stereo.Audio:0" {type=Audio}
 "UE.OutputFormat.Stereo.Audio:1" {type=Audio}
]
pages [
 Default {
  is_build_graph=true
  input_defaults {
   BaseVolume=0.612
   Character.Affiliation=0
   Character.Health=0
   Character.UniqueID=0
   Loop.Pitch.Base=0
   Loop.Pitch.Range=0
   Loop.Volume.Base=-3.44
   Loop.Volume.Range=0
   Start.Pitch.Base=1.3
   Start.Pitch.Range=0.1
   Start.Sounds=null
   Start.Volume.Base=-6.576
   Start.Volume.Range=0.1
   UE.Source.OnPlay=null
   Loop.Sounds []
  }
  nodes [
   n1 = AbsoluteValue.Abs:Float(Input=^Loop.Volume.Range)
   DecibelstoLinearGain_1 = UE.Decibels to Linear Gain:Float(Decibels=UE.RandomFloat(
     Next=^UE.Source.OnPlay
     Max=UE.Add:Float(PrimaryOperand=^Loop.Volume.Base, AdditionalOperands=@n1.Value).Out
     Min=UE.Subtract:Float(PrimaryOperand=^Loop.Volume.Base, AdditionalOperands=@n1.Value).Out
    ).Value)
   n2 = InitVariable.Float()
   n3 = AbsoluteValue.Abs:Float(Input=^Start.Volume.Range)
   DecibelstoLinearGain_0 = UE.Decibels to Linear Gain:Float(Decibels=UE.RandomFloat(
     Next=^UE.Source.OnPlay
     Max=UE.Add:Float(PrimaryOperand=^Start.Volume.Base, AdditionalOperands=@n3.Value).Out
     Min=UE.Subtract:Float(PrimaryOperand=^Start.Volume.Base, AdditionalOperands=@n3.Value).Out
    ).Value)
   n4 = InitVariable.Float()
   TriggerDelay_0 = UE.Trigger Delay(In=^UE.Source.OnPlay) {also {defaults {Delay Time=0.3}}}
   ArrayRandomGet_1 = Array.Random Get:WaveAsset:Array(In Array=^Loop.Sounds, Next=@TriggerDelay_0.Out) {also {defaults {Weights []}}}
   n5 = InitVariable.Float()
   n6 = AbsoluteValue.Abs:Float(Input=^Loop.Pitch.Range)
   n7 = UE.RandomFloat(
    Next=^UE.Source.OnPlay
    Max=UE.Add:Float(PrimaryOperand=^Loop.Pitch.Base, AdditionalOperands=@n6.Value).Out
    Min=UE.Subtract:Float(PrimaryOperand=^Loop.Pitch.Base, AdditionalOperands=@n6.Value).Out
   )
   WavePlayer_0 = UE.Wave Player:Stereo(
    Play=@ArrayRandomGet_1.On Next
    Wave Asset=@ArrayRandomGet_1.Value
    Pitch Shift=VariableAccessor.Float(Variable=VariableMutator.Float(Variable=@n5.Variable, Value=@n7.Value).Variable).Value
   ) {also {defaults {Loop="<MetaSoundBuilderResult.SUCCEEDED: 0>"}, version [1,1]}}
   ArrayRandomGet_0 = Array.Random Get:WaveAsset:Array(In Array=^Start.Sounds, Next=^UE.Source.OnPlay)
   n8 = InitVariable.Float()
   n9 = AbsoluteValue.Abs:Float(Input=^Start.Pitch.Range)
   n10 = UE.RandomFloat(
    Next=^UE.Source.OnPlay
    Max=UE.Add:Float(PrimaryOperand=^Start.Pitch.Base, AdditionalOperands=@n9.Value).Out
    Min=UE.Subtract:Float(PrimaryOperand=^Start.Pitch.Base, AdditionalOperands=@n9.Value).Out
   )
   WavePlayer_1 = UE.Wave Player:Stereo(
    Play=@ArrayRandomGet_0.On Next
    Wave Asset=@ArrayRandomGet_0.Value
    Pitch Shift=VariableAccessor.Float(Variable=VariableMutator.Float(Variable=@n8.Variable, Value=@n10.Value).Variable).Value
   ) {also {version [1,1]}}
   AudioMixerAudioMixerStereo2_0 = "AudioMixer.Audio Mixer (Stereo, 2)"(
    Gain 1=VariableAccessor.Float(Variable=VariableMutator.Float(Value=@DecibelstoLinearGain_1.Linear Gain, Variable=@n2.Variable).Variable).Value
    Gain 0=VariableAccessor.Float(Variable=VariableMutator.Float(Value=@DecibelstoLinearGain_0.Linear Gain, Variable=@n4.Variable).Variable).Value
    In 1 L=@WavePlayer_0.Out Left
    In 1 R=@WavePlayer_0.Out Right
    In 0 L=@WavePlayer_1.Out Left
    In 0 R=@WavePlayer_1.Out Right
   )
   Multiply_0 = UE.Multiply:Audio by Float(AdditionalOperands=^BaseVolume, PrimaryOperand=@AudioMixerAudioMixerStereo2_0.Out L)
   Multiply_1 = UE.Multiply:Audio by Float(AdditionalOperands=^BaseVolume, PrimaryOperand=@AudioMixerAudioMixerStereo2_0.Out R)
  ]
  links [{dst="@UE.OutputFormat.Stereo.Audio:0", src=Multiply_0.Out}, {dst="@UE.OutputFormat.Stereo.Audio:1", src=Multiply_1.Out}]
 }
 Mobile {
  nodes [
   ArrayRandomGet_0 = Array.Random Get:WaveAsset:Array(In Array=^Loop.Sounds, Next=^UE.Source.OnPlay) {also {defaults {Weights []}}}
   n1 = InitVariable.Float()
   n2 = AbsoluteValue.Abs:Float(Input=^Loop.Pitch.Range)
   n3 = UE.RandomFloat(
    Next=^UE.Source.OnPlay
    Max=UE.Add:Float(PrimaryOperand=^Loop.Pitch.Base, AdditionalOperands=@n2.Value).Out
    Min=UE.Subtract:Float(PrimaryOperand=^Loop.Pitch.Base, AdditionalOperands=@n2.Value).Out
   )
   WavePlayer_0 = UE.Wave Player:Mono(
    Play=@ArrayRandomGet_0.On Next
    Wave Asset=@ArrayRandomGet_0.Value
    Pitch Shift=VariableAccessor.Float(Variable=VariableMutator.Float(Variable=@n1.Variable, Value=@n3.Value).Variable).Value
   ) {also {defaults {Loop="<MetaSoundBuilderResult.SUCCEEDED: 0>"}, version [1,1]}}
   Multiply_0 = UE.Multiply:Audio by Float(AdditionalOperands=^BaseVolume, PrimaryOperand=@WavePlayer_0.Out Mono)
   Multiply_1 = UE.Multiply:Audio by Float(AdditionalOperands=^BaseVolume, PrimaryOperand=@WavePlayer_0.Out Mono)
   n4 = AbsoluteValue.Abs:Float(Input=^Start.Volume.Range)
   DecibelstoLinearGain_0 = UE.Decibels to Linear Gain:Float(Decibels=UE.RandomFloat(
     Next=^UE.Source.OnPlay
     Max=UE.Add:Float(PrimaryOperand=^Start.Volume.Base, AdditionalOperands=@n4.Value).Out
     Min=UE.Subtract:Float(PrimaryOperand=^Start.Volume.Base, AdditionalOperands=@n4.Value).Out
    ).Value)
   VariableMutator.Float(Value=@DecibelstoLinearGain_0.Linear Gain, Variable=InitVariable.Float().Variable)
   n5 = AbsoluteValue.Abs:Float(Input=^Loop.Volume.Range)
   DecibelstoLinearGain_1 = UE.Decibels to Linear Gain:Float(Decibels=UE.RandomFloat(
     Next=^UE.Source.OnPlay
     Max=UE.Add:Float(PrimaryOperand=^Loop.Volume.Base, AdditionalOperands=@n5.Value).Out
     Min=UE.Subtract:Float(PrimaryOperand=^Loop.Volume.Base, AdditionalOperands=@n5.Value).Out
    ).Value)
   VariableMutator.Float(Value=@DecibelstoLinearGain_1.Linear Gain, Variable=InitVariable.Float().Variable)
   n6 = AbsoluteValue.Abs:Float(Input=^Start.Pitch.Range)
   VariableMutator.Float(
    Variable=InitVariable.Float().Variable
    Value=UE.RandomFloat(
     Next=^UE.Source.OnPlay
     Max=UE.Add:Float(PrimaryOperand=^Start.Pitch.Base, AdditionalOperands=@n6.Value).Out
     Min=UE.Subtract:Float(PrimaryOperand=^Start.Pitch.Base, AdditionalOperands=@n6.Value).Out
    ).Value
   )
  ]
  links [{dst="@UE.OutputFormat.Stereo.Audio:0", src=Multiply_0.Out}, {dst="@UE.OutputFormat.Stereo.Audio:1", src=Multiply_1.Out}]
 }
]
properties {
 assetImportData=&self:AssetImportData
 bLooping=true
 duration=10000
 numChannels=2
 outputFormat=Stereo
 sampleRate=48000
}
