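# Looping stereo MetaSound Source preset for friendly metal level 1 building construction audio, derived from MSS_BuildingLoop.
# It only overrides parent inputs: three Construction_Done metal waves fill the Start slot and one level 01 construction wave fills the Loop slot, each with volume and pitch settings.
# The friendly audio properties sheet, wall-build attenuation, player wall-build concurrency and player construction sound class are assigned.
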MetaSoundSource /Game/Sounds/Fort_UI_Sounds/Gameplay/BuildingSounds/MSS/Friendly/MSS_BuildingLoop_Metal_Level1_Friendly
preset_parent=/Game/Sounds/Fort_UI_Sounds/Gameplay/BuildingSounds/MSS/MSS_BuildingLoop
inputs [
 BaseVolume {type=Float}
 Character.Affiliation {type=Enum:FortAudioAffiliation}
 Character.Health {type=Float}
 Character.UniqueID {type=Int32}
 Loop.Pitch.Base {type=Float}
 Loop.Pitch.Range {type=Float}
 Loop.Sounds {type=WaveAsset:Array}
 Loop.Volume.Base {type=Float}
 Loop.Volume.Range {type=Float}
 Start.Pitch.Base {type=Float}
 Start.Pitch.Range {type=Float}
 Start.Sounds {type=WaveAsset:Array}
 Start.Volume.Base {type=Float}
 Start.Volume.Range {type=Float}
 TESTLOOP {type=WaveAsset:Array}
 TESTSTART {type=WaveAsset:Array}
 UE.Source.OnPlay {type=Trigger}
]
interfaces [Fort.Character:1.0, UE.OutputFormat.Stereo:1.0, UE.Source:1.1]
outputs [
 "UE.OutputFormat.Stereo.Audio:0" {type=Audio}
 "UE.OutputFormat.Stereo.Audio:1" {type=Audio}
]
pages [Default {input_defaults {
   BaseVolume=0.612
   Character.Affiliation=0
   Character.Health=0
   Character.UniqueID=0
   Loop.Pitch.Base=0
   Loop.Pitch.Range=0
   Loop.Volume.Base=-3.44
   Loop.Volume.Range=0
   Start.Pitch.Base=1.3
   Start.Pitch.Range=0.1
   Start.Volume.Base=-6.576
   Start.Volume.Range=0.1
   UE.Source.OnPlay=null
   Loop.Sounds [/Game/Sounds/Fort_UI_Sounds/Gameplay/BuildingSounds/Wavs/metal_level_const_level_01_A]
   Start.Sounds [
    /Game/Sounds/Fort_UI_Sounds/Gameplay/BuildingSounds/Wavs/Fort_Construction_Done_Metal_01
    /Game/Sounds/Fort_UI_Sounds/Gameplay/BuildingSounds/Wavs/Fort_Construction_Done_Metal_02
    /Game/Sounds/Fort_UI_Sounds/Gameplay/BuildingSounds/Wavs/Fort_Construction_Done_Metal_03
   ]
   TESTLOOP [&metal_level_const_level_01_A]
   TESTSTART [&Fort_Construction_Done_Metal_01]
  }}]
properties {
 assetImportData=&self:AssetImportData
 attenuationSettings=/Game/Sounds/Attenuation/Fort_Build_Wall_Att
 audioPropertiesSheet=/Game/Sounds/Fort_UI_Sounds/Gameplay/BuildingSounds/Settings/APS_BuildingLoop_Friendly
 bLooping=true
 duration=10000
 maxDistance=2000
 numChannels=2
 outputFormat=Stereo
 sampleRate=48000
 soundClassObject=/Game/Sounds/FortSoundClassesAndModes/Fort_Player_Construction
 concurrencySet [/Game/Sounds/Concurrency_Rules/Player_Wall_Build_Conc]
}
