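# Stereo MetaSound Source preset for the looping sound of friendly level 3 stone building construction.
# It inherits its graph from MSS_BuildingLoop, overriding start and loop waves, pitch and volume.
# Start uses three stone "construction done" variants with raised pitch and lowered volume, while the loop reuses a level 2 rock construction wave.
# Wall-build attenuation, player wall-build concurrency, the player construction sound class and a friendly audio properties sheet govern playback.
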MetaSoundSource /Game/Sounds/Fort_UI_Sounds/Gameplay/BuildingSounds/MSS/Friendly/MSS_BuildingLoop_Stone_Level3_Friendly
preset_parent=/Game/Sounds/Fort_UI_Sounds/Gameplay/BuildingSounds/MSS/MSS_BuildingLoop
inputs [
 Loop.Pitch.Base {type=Float}
 Loop.Pitch.Range {type=Float}
 Loop.Sounds {type=WaveAsset:Array}
 Loop.Volume.Base {type=Float}
 Loop.Volume.Range {type=Float}
 Start.Pitch.Base {type=Float}
 Start.Pitch.Range {type=Float}
 Start.Sounds {type=WaveAsset:Array}
 Start.Volume.Base {type=Float}
 Start.Volume.Range {type=Float}
 UE.Source.OnPlay {type=Trigger}
]
interfaces [UE.OutputFormat.Stereo:1.0, UE.Source:1.1]
outputs [
 "UE.OutputFormat.Stereo.Audio:0" {type=Audio}
 "UE.OutputFormat.Stereo.Audio:1" {type=Audio}
]
pages [Default {input_defaults {
   Loop.Pitch.Base=0
   Loop.Pitch.Range=0
   Loop.Volume.Base=0
   Loop.Volume.Range=0
   Start.Pitch.Base=1.3
   Start.Pitch.Range=0.1
   Start.Volume.Base=-4
   Start.Volume.Range=0.1
   UE.Source.OnPlay=null
   Loop.Sounds [/Game/Sounds/Fort_UI_Sounds/Gameplay/BuildingSounds/Wavs/rock_lev_02_construction_loop_C]
   Start.Sounds [
    /Game/Sounds/Fort_UI_Sounds/Gameplay/BuildingSounds/Wavs/Fort_Construction_Done_Stone_01
    /Game/Sounds/Fort_UI_Sounds/Gameplay/BuildingSounds/Wavs/Fort_Construction_Done_Stone_02
    /Game/Sounds/Fort_UI_Sounds/Gameplay/BuildingSounds/Wavs/Fort_Construction_Done_Stone_03
   ]
  }}]
properties {
 assetImportData=&self:AssetImportData
 attenuationSettings=/Game/Sounds/Attenuation/Fort_Build_Wall_Att
 audioPropertiesSheet=/Game/Sounds/Fort_UI_Sounds/Gameplay/BuildingSounds/Settings/APS_BuildingLoop_Friendly
 bLooping=true
 duration=10000
 maxDistance=2000
 numChannels=2
 outputFormat=Stereo
 sampleRate=48000
 soundClassObject=/Game/Sounds/FortSoundClassesAndModes/Fort_Player_Construction
 concurrencySet [/Game/Sounds/Concurrency_Rules/Player_Wall_Build_Conc]
}
